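# Unlit translucent material for gizmos in the Skeletal Mesh Modeling Tools editor plugin.
# Its only logic scales the RGB of the GizmoColor vector parameter, white by default, by a constant to brighten it.
# Disabled depth testing, two-sided rendering and editor compositing let it draw over scene geometry, and it casts no ray-traced shadows.
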Material /SkeletalMeshModelingTools/SkeletalMeshGizmoMaterial
graph {
 properties {
  bCastRayTracedShadows=false
  bDisableDepthTest=true
  bUsedWithEditorCompositing=true
  blendMode=BLEND_Translucent
  pixelDepthOffsetMode=PDOM_Legacy
  shadingModel=MSM_Unlit
  twoSided=true
  usedShadingModels=Unlit
 }
 nodes [Multiply(A=VectorParameter() {parameterName=GizmoColor, defaultValue=<rgba 1 1 1 1>}.RGB) {constB=1.5}]
}
